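; Fake light prop actor that configures a decal component, apparently to imitate lighting, from a decal settings struct.
; ApplyLightSettings runs in both the construction script and BeginPlay.
; It finds the DecalComponent by class and, if valid, sets visibility, material, color, sort order, fade screen size and fade in/out, including optional destruction after fade.
; Tick and overlap events are empty.
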(event EventBeginPlay
  (FakeLightSettings|ApplyLightSettings))

(event Collision|EventActorBeginOverlap
  (params (OtherActor :type Actor%20Object%20Reference)))

(event EventTick
  (params (DeltaSeconds :type Float%20%28single-precision%29)))

(fn ApplyLightSettings
  (bind (_decalsettings_isvisible_1_88f7c47c44c26cc98b815d8ff93c6329 _decalsettings_materialinstance_4_4d43826e40a05c8acf1c409f7579a5db _decalsettings_decalcolor_7_11356b7e4568f8aec6890890ceb700b6 _decalsettings_sortorder_10_506493684c41ab08b0f5d795fbacd473 _decalsettings_fadescreensize_13_1706df5f448e4543a66b759c4a81f6de _decalsettings_fadein_16_bd2f898d48c8b5f030737ebc6cb2292f _decalsettings_fadeout_18_698460f0407d2458c63c6096e0590113 _decalsettings_destroyobjectafterfade_21_fd8cb0d947e639649855238b62ca2918) (Utilities|Struct|BreakSTRFakeLightDecal02 (Variables|GetDecalSettings_IsVisible_1_88F7C47C44C26CC98B815D8FF93C6329)))
  (bind (_x _y) (Math|Vector2D|BreakVector2D _decalsettings_fadein_16_bd2f898d48c8b5f030737ebc6cb2292f))
  (bind _returnvalue (Actor|GetComponentbyClass "/Script/Engine.DecalComponent"))
  (bind (_x_2 _y_2) (Math|Vector2D|BreakVector2D _decalsettings_fadeout_18_698460f0407d2458c63c6096e0590113))
  (Utilities|IsValid _returnvalue
    (:Is%20Valid
      (Rendering|SetVisibility _returnvalue _decalsettings_isvisible_1_88f7c47c44c26cc98b815d8ff93c6329 true)
      (Rendering|Components|Decal|SetDecalMaterial _returnvalue _decalsettings_materialinstance_4_4d43826e40a05c8acf1c409f7579a5db)
      (Rendering|Components|Decal|SetDecalColor _returnvalue _decalsettings_decalcolor_7_11356b7e4568f8aec6890890ceb700b6)
      (Rendering|Components|Decal|SetSortOrder _returnvalue _decalsettings_sortorder_10_506493684c41ab08b0f5d795fbacd473)
      (Rendering|Components|Decal|SetFadeScreenSize _returnvalue _decalsettings_fadescreensize_13_1706df5f448e4543a66b759c4a81f6de)
      (Rendering|Components|Decal|SetFadeIn _returnvalue _x _y)
      (Rendering|Components|Decal|SetFadeOut _returnvalue _x_2 _y_2 _decalsettings_destroyobjectafterfade_21_fd8cb0d947e639649855238b62ca2918))))

(fn ConstructionScript
  (FakeLightSettings|ApplyLightSettings))
