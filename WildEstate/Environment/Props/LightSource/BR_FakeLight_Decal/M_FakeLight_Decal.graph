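# Deferred decal material, modulate blend, that fakes a cone-shaped light cast onto scene surfaces.
# UV-space inner and outer cone angles, falloff and hotspot shape the light; Decal Color supplies tint, or blackbody temperature, and intensity.
# Scene world normals, transformed against decal axes, drive directional highlight and side/back occlusion masks.
# Time-of-day collection values and optional Perlin-noise flicker modulate it, while platform switches provide simpler mobile color and opacity.
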Material /WildEstate/Environment/Props/LightSource/BR_FakeLight_Decal/M_FakeLight_Decal
graph {
 properties {
  bAutomaticallySetUsageInEditor=false
  bCastRayTracedShadows=false
  blendMode=BLEND_Modulate
  materialDomain=MD_DeferredDecal
  usedShadingModels=Default Lit
 }
 nodes [
  n1 = NamedRerouteUsage()
  n2 = Add(A=TextureCoordinate()) {constB=-0.5}
  n3 = NamedRerouteUsage()
  Add(A=AppendVector(A=DotProduct(A=@n1, B=@n2), B=DotProduct(A=@n2, B=@n3))) {constB=0.5}
  Constant() {r=1}
  n4 = LinearInterpolate(B=NamedRerouteUsage(), Alpha=ScalarParameter() {defaultValue=1, group=Time Of Day, parameterName=TOD Influence, sliderMax=1}) {constA=1}
  LinearInterpolate(B=NamedRerouteUsage(), Alpha=@n4)
  n5 = NamedRerouteUsage()
  n6 = Constant()
  MaterialFunctionCall(
   VectorToTransform=Constant3Vector() {desc=Light Direction, constant=<rgba 0 -1 0 0>}
   BasisX=Constant3Vector() {constant=<rgba 0 0 1 0>}
   BasisY=AppendVector(A=NamedRerouteUsage(), B=Constant())
   BasisZ=Multiply(A=AppendVector(A=@n5, B=@n6)) {constB=-1}
  ) {desc=To Rotated UV Space, materialFunction=&Transform3x3Matrix}
  n7 = VectorParameter() {parameterName=awehrg4, defaultValue=<rgba 8 8 0.5 7>}
  n8 = ScreenPosition()
  n9 = Multiply(A=AppendVector(A=@n7.R, B=@n7.R), B=@n7.B)
  MaterialFunctionCall!(
   Vector3
   MaximumNumberOfDigits=@n7.A
   UVs=Frac(None=Multiply(A=@n8.ViewportUV, B=@n9))
   DebugTextLocation RG_UpperRight BA_LowerLeft
   Component Spacing
  ) {materialFunction=/Engine/Functions/Engine_MaterialFunctions02/Utility/DebugFloat3Values}
  MaterialFunctionCall() {materialFunction=&DebugScalarValues}
  n10 = NamedRerouteUsage()
  Multiply(
   A=MaterialFunctionCall(Emissive=LinearInterpolate(B=@n10, Alpha=@n4)) {materialFunction=&MF_FX_EmissiveAdjustments}.Result
   B=NamedRerouteUsage()
  )
  n11 = MaterialFunctionCall(Float2=NamedRerouteUsage()) {materialFunction=&BreakOutFloat2Components}
  n12 = Add(A=Multiply(A=@n11.R) {constB=2}) {constB=-1}
  n13 = NamedRerouteUsage()
  n14 = Subtract(A=OneMinus(None=Abs(None=@n12)), B=@n13)
  n15 = Power(
   Base=MaterialFunctionCall(Input=Saturate(None=@n14), Input High=OneMinus(None=@n13)) {materialFunction=&RemapValueRange}.Result
   Exp=NamedRerouteUsage()
  )
  n16 = NamedRerouteUsage()
  n17 = Subtract(A=@n11.G, B=@n16)
  n18 = Power(
   Base=MaterialFunctionCall(Input=Saturate(None=@n17), Input High=OneMinus(None=@n16)) {materialFunction=&RemapValueRange}.Result
   Exp=NamedRerouteUsage()
  )
  Multiply(A=NamedRerouteUsage(), B=Multiply(A=Saturate(None=@n15), B=Saturate(None=@n18)))
  n19 = Constant4Vector() {constant=<rgba 0 1 0.5 0.74>}
  n20 = MaterialFunctionCall(Input=NamedRerouteUsage(), Input Low=@n19.R, Input High=@n19.R, Target Low=@n19.R, Target High=@n19.R) {materialFunction=&RemapValueRange}
  NamedRerouteDeclaration(None=AppendVector(A=Sine(None=@n20.Result), B=Cosine(None=@n20.Result))) {name=SinCos Cone, nodeColor=<rgba 1 0 0.0938 1>}
  n21 = Subtract(A=NamedRerouteUsage(), B=NamedRerouteUsage())
  n22 = Divide(A=@n21, B=NamedRerouteUsage())
  n23 = OneMinus(None=Multiply(A=Saturate(None=@n22), B=Step(X=@n21)))
  n24 = NamedRerouteUsage()
  n25 = Add(A=Saturate(None=Multiply(A=@n23, B=@n24)) {desc=Inner Cone}, B=NamedRerouteUsage())
  n26 = NamedRerouteUsage()
  NamedRerouteDeclaration(None=Saturate(None=Multiply(A=@n25, B=@n26))) {name=Cone Mask, nodeColor=<rgba 0.7344 1 0 1>}
  n27 = ComponentMask(None=SceneTexture() {sceneTextureId=PPI_WorldNormal}.Color) {b=true, g=true, r=true}
  n28 = NamedRerouteUsage()
  n29 = Normalize(VectorInput=Transform(None=Constant3Vector() {constant=<rgba 0 -1 0 0>}) {transformSourceType=TRANSFORMSOURCE_Local})
  n30 = Normalize(VectorInput=Transform(None=Constant3Vector() {constant=<rgba 0 0 -1 0>}) {transformSourceType=TRANSFORMSOURCE_Local})
  n31 = Normalize(VectorInput=Transform(None=Constant3Vector() {constant=<rgba -1 0 0 0>}) {transformSourceType=TRANSFORMSOURCE_Local})
  n32 = Abs(None=DotProduct(
    A=@n27
    B=MaterialFunctionCall(VectorToTransform=@n28, BasisX=@n29, BasisY=@n30, BasisZ=@n31) {materialFunction=&Transform3x3Matrix}.Transformed Vector
   ))
  NamedRerouteDeclaration(None=Saturate(None=SmoothStep(Value=@n32) {constMax=0.5, constMin=1, desc=Anlge})) {name=Occlusion Mask Sides, nodeColor=<rgba 0 1 0.8594 1>}
  n33 = MaterialFunctionCall(VectorToTransform=@n28, BasisX=@n30, BasisY=@n29, BasisZ=@n31) {
   desc="To Decal Oriented World Space\r\n\r\nDecal points to X- hence swizzle"
   materialFunction=&Transform3x3Matrix
  }
  NamedRerouteDeclaration(None=Saturate(None=DotProduct(A=@n33.Transformed Vector, B=@n27))) {name=Highlight Mask, nodeColor=<rgba 0 0.5078 1 1>}
  n34 = OneMinus(None=Length(None=NamedRerouteUsage()) {is=class MaterialExpressionLength})
  n35 = Saturate(None=Divide(A=@n34, B=NamedRerouteUsage()) {constB=0.95})
  n36 = Power(Base=@n35, Exp=NamedRerouteUsage()) {constExponent=4.5}
  n37 = SmoothStep(Value=@n34) {constMax=0.2}
  n38 = ComponentMask(None=NamedRerouteUsage()) {r=true}
  n39 = SquareRoot(None=Sine(None=Multiply(A=@n38) {constB=0.5}))
  NamedRerouteDeclaration(None=Multiply(A=LinearInterpolate(A=@n35, B=@n36, Alpha=@n37), B=Saturate(None=@n39))) {name=Cone Falloff, nodeColor=<rgba 0 1 0.1328 1>}
  NamedRerouteDeclaration() {name=Light Rotation Angle, nodeColor=<rgba 1 0.375 0 1>}
  n40 = VectorParameter() {
   group=001 - Cone Settings
   parameterName=Cone Shape Controls OLD
   sortPriority=0
   channelNames {a=Cone Falloff, b=Cone Height, g=Light Source Shape, r=Cone Angle}
   defaultValue=<rgba 0.75 0.4 1 0.35>
  }
  NamedRerouteDeclaration(None=Saturate(None=@n40.Cone Angle)) {name=Cone Angle, nodeColor=<rgba 0.0547 0 1 1>}
  NamedRerouteDeclaration(None=Saturate(None=@n40.Light Source Shape)) {name=Light Source Shape, nodeColor=<rgba 0.2656 1 0 1>}
  NamedRerouteDeclaration(None=Saturate(None=@n40.Cone Height)) {name=Cone Height, nodeColor=<rgba 1 0 0.5859 1>}
  n41 = NamedRerouteUsage()
  n42 = NamedRerouteDeclaration(None=AppendVector(A=Sine(None=@n41), B=Cosine(None=@n41))) {name=SinCos Rotation, nodeColor=<rgba 0 1 0.6016 1>}
  NamedRerouteDeclaration(None=Multiply(
    A=MaterialFunctionCall(XY=@n42) {materialFunction=&Swizzle}.YX
    B=Constant2Vector() {g=-1, r=1}
   )) {name=Cos-Sin Rotation, nodeColor=<rgba 0 0.9062 1 1>}
  n43 = CollectionParameter() {
   collection=&FortniteMaterialParameters
   parameterName=FortniteActiveTimeOfDay
  }
  n44 = MaterialFunctionCall(Float4=@n43) {materialFunction=&BreakOutFloat4Components}
  n45 = OneMinus(None=@n44.A)
  n46 = Add(A=Multiply(A=@n44.R) {constB=0.25}, B=Multiply(A=@n44.B) {constB=0.25})
  n47 = Saturate(None=ScalarParameter() {defaultValue=1, group=Time Of Day, parameterName=Gradual Light Transition, sliderMax=1, sortPriority=0})
  NamedRerouteDeclaration(None=LinearInterpolate(A=Step(Y=@n45) {constX=0.01, constY=0.01}, B=Add(A=@n46, B=@n44.A), Alpha=Round(None=@n47))) {name=TOD, nodeColor=<rgba 0 1 0.1797 1>}
  n48 = DecalColor()
  n49 = Min(A=@n48.R) {constB=12000}
  n50 = Normalize(VectorInput=BlackBody!(Temp=Max(A=@n49) {constB=1700}))
  n51 = Saturate(None=ScalarParameter() {defaultValue=1, group=Fake Light Settings, parameterName=Use Temperature, sliderMax=1, sortPriority=3})
  NamedRerouteDeclaration(None=Multiply(A=LinearInterpolate(A=@n48.RGB, B=@n50, Alpha=@n51), B=@n48.A)) {name=Light Color, nodeColor=<rgba 0.9453 0 1 1>}
  NamedRerouteDeclaration(None=Constant3Vector() {constant=<rgba 0 -1 0 0>}) {name=Fake Light Dir, nodeColor=<rgba 0 1 0.4609 1>}
  n52 = NamedRerouteDeclaration(None=TextureCoordinate()) {name=Raw UV, nodeColor=<rgba 0 1 0.625 1>}
  n53 = Constant2Vector() {g=-1, r=-0.5}
  n54 = AppendVector(A=NamedRerouteDeclaration(None=Add(A=@n52, B=@n53) {constB=-0.5}) {name=Light UVs, nodeColor=<rgba 0.0312 1 0 1>}, B=NamedRerouteUsage())
  NamedRerouteDeclaration(None=DotProduct(A=Normalize(VectorInput=@n54), B=NamedRerouteUsage())) {name=CosTheta, nodeColor=<rgba 0 1 0.6953 1>}
  n55 = NamedRerouteUsage()
  n56 = NamedRerouteDeclaration(None=Multiply(A=NamedRerouteUsage(), B=NamedRerouteUsage())) {name=Final Color, nodeColor=<rgba 0 0.7891 1 1>}
  n57 = ScalarParameter() {defaultValue=8, group=Fake Light Settings, parameterName=Directional Highlight Intensity, sliderMax=10, sliderMin=1, sortPriority=15}
  n58 = Multiply(A=@n56, B=@n57) {constB=2}
  n59 = NamedRerouteUsage()
  n60 = LinearInterpolate(
   A=LinearInterpolate(A=Multiply(A=@n55) {constB=0.5}, B=LinearInterpolate(A=@n56, B=@n58, Alpha=@n59) {constB=0}, Alpha=NamedRerouteUsage()) {constB=0}
   Alpha=NamedRerouteUsage()
  ) {constB=0}
  n61 = Desaturation(
   None=Multiply(A=NamedRerouteUsage(), B=ScalarParameter() {defaultValue=3, group=Fake Light Settings, parameterName=Mobile Color Multiplier}) {constB=3}
   Fraction=Constant() {r=-0.1}
  )
  NamedRerouteDeclaration(None=MaterialFunctionCall!("Xb1,PS4"=@n60, "PC-High&&Epic,XSX,PS5"=@n60, "PC-Perf,Android"=@n61, Switch=@n61, "PC-Low(NotPerf)&&Medium"=@n60) {materialFunction=/Game/Effects/Niagara/MasterMaterials/Functions/Common/PlatformSwitch_3D}.Result) {name=Platform Adjusted Final Color, nodeColor=<rgba 0 1 0.4375 1>}
  n62 = ScalarParameter() {defaultValue=80, group=Fake Light Settings, parameterName=Outer Cone Angle, sliderMax=90, sliderMin=1, sortPriority=1}
  n63 = Divide() {constB=360}
  n64 = Multiply(A=Round(None=@n62), B=@n63) {constB=0.0028, desc=Degrees to turns}
  NamedRerouteDeclaration(None=Min(A=Cosine(None=@n64), B=NamedRerouteUsage())) {name=Cos Outer, nodeColor=<rgba 1 0.9844 0 1>}
  n65 = ScalarParameter() {defaultValue=30, group=Fake Light Settings, parameterName=Inner Cone Angle, sliderMax=85, sliderMin=1, sortPriority=0}
  n66 = Round(None=@n65)
  NamedRerouteDeclaration(None=Cosine(None=Multiply(A=@n66, B=@n63) {constB=0.0028})) {name=Cos Inner, nodeColor=<rgba 0.6641 0 1 1>}
  n67 = NamedRerouteUsage()
  n68 = NamedRerouteUsage()
  n69 = NamedRerouteUsage()
  NamedRerouteDeclaration(None=Saturate(None=SmoothStep(Min=@n67, Max=@n68, Value=@n69))) {name=Cone Base, nodeColor=<rgba 0 1 0.2969 1>}
  n70 = NamedRerouteUsage()
  n71 = LinearInterpolate(A=LinearInterpolate(A=Multiply(A=@n70) {constB=0.35}, B=@n70, Alpha=NamedRerouteUsage()) {constB=0}, Alpha=NamedRerouteUsage()) {constB=0}
  n72 = NamedRerouteUsage()
  n73 = NamedRerouteUsage()
  n74 = Constant() {r=1}
  n75 = Saturate(None=Multiply(A=Multiply(A=@n72) {constB=0.5}, B=Desaturation(None=@n73, Fraction=@n74)))
  NamedRerouteDeclaration(None=MaterialFunctionCall("Xb1,PS4"=@n71, "PC-High&&Epic,XSX,PS5"=@n71, "PC-Perf,Android"=@n75, Switch=@n75, "PC-Low(NotPerf)&&Medium"=@n71) {materialFunction=&PlatformSwitch_1D}.Result) {name=Platform Adjusted Opacity Mask, nodeColor=<rgba 1 0.1172 0 1>}
  n76 = ScalarParameter() {defaultValue=0.5, group=Fake Light Settings, parameterName=Inner Cone Softness, sliderMax=1, sortPriority=5}
  n77 = Constant2Vector() {g=0.1, r=0}
  NamedRerouteDeclaration(None=MaterialFunctionCall(Input=Saturate(None=@n76), Target Low=@n77.R, Target High=@n77.R) {materialFunction=&RemapValueRange}.Result) {name=Inner Cone Delta, nodeColor=<rgba 1 0.6797 0 1>}
  n78 = ScalarParameter() {defaultValue=0.2, group=Fake Light Settings, parameterName=Cone Origin Softness, sliderMax=1, sortPriority=6}
  n79 = Constant2Vector() {g=0.1}
  NamedRerouteDeclaration(None=MaterialFunctionCall(Input=Saturate(None=@n78), Target Low=@n79.R, Target High=@n79.R) {materialFunction=&RemapValueRange}.Result) {name=Cone Depth, nodeColor=<rgba 0 1 0.4141 1>}
  n80 = Divide(
   A=ComponentMask(None=ObjectPositionWS()) {g=true, r=true}
   B=ScalarParameter() {defaultValue=50, group=Flickering, parameterName=Flicker UV Divide, sortPriority=34}
  ) {constB=100}
  n81 = Multiply(A=Time(), B=ScalarParameter() {defaultValue=1, group=Flickering, parameterName=Flicker Speed, sortPriority=35}) {constB=2}
  n82 = MaterialFunctionCall(
   In=TextureSample(UVs=Add(A=@n80, B=@n81)) {samplerType=SAMPLERTYPE_LinearGrayscale, texture=&T_SimplePerlin_Greyscale}.R
   Contrast=ScalarParameter() {group=Flickering, parameterName=Flicker Contrast, sortPriority=36}
  ) {materialFunction=&CheapContrast}
  NamedRerouteDeclaration(None=StaticSwitchParameter(True=Saturate(None=@n82.Result), False=Constant() {r=1}) {group=Flickering, parameterName=Flickering, sortPriority=33}) {name=Flickering, nodeColor=<rgba 0.9219 1 0 1>}
  n83 = ScalarParameter() {defaultValue=0.5, group=Fake Light Settings, parameterName=Inner Cone Weight, sliderMax=1, sortPriority=4}
  n84 = Constant2Vector() {g=0.25}
  NamedRerouteDeclaration(None=MaterialFunctionCall(Input=Saturate(None=@n83), Target Low=@n84.R, Target High=@n84.R) {materialFunction=&RemapValueRange}.Result) {name=Inner Cone Weight, nodeColor=<rgba 1 0.8672 0 1>}
  n85 = ScalarParameter() {defaultValue=0.25, group=Fake Light Settings, parameterName=Hotspot Intensity, sliderMax=1, sortPriority=7}
  n86 = Constant2Vector() {g=0.9, r=1}
  NamedRerouteDeclaration(None=MaterialFunctionCall(Input=Saturate(None=@n85), Target Low=@n86.R, Target High=@n86.R) {materialFunction=&RemapValueRange}.Result) {name=Hotspot, nodeColor=<rgba 0 1 0.6484 1>}
  NamedRerouteDeclaration(None=Multiply(A=ScalarParameter() {defaultValue=3.5, group=Fake Light Settings, parameterName=Light Falloff Exponent, sliderMax=50, sortPriority=2}) {constB=0.3828}) {name=Light Falloff Exponent, nodeColor=<rgba 1 0.6094 0 1>}
  n87 = Multiply(A=@n33.Transformed Vector) {constB=-1}
  NamedRerouteDeclaration(None=Saturate(None=DotProduct(A=@n27, B=@n87))) {name=Occlusion Mask Back, nodeColor=<rgba 0 1 0.0156 1>}
  n88 = VectorParameter() {
   group=002 - Opacity Settings
   parameterName=Cone Fade Controls OLD
   sortPriority=0
   channelNames {a=Y Softness, b=Y Offset, g=X Softness, r=X Offset}
   defaultValue=<rgba 0 1 0 2>
  }
  NamedRerouteDeclaration(None=@n88.X Offset) {name=Mask X Offset, nodeColor=<rgba 0 1 0.8125 1>}
  NamedRerouteDeclaration(None=@n88.Y Offset) {name=Mask Y Offset, nodeColor=<rgba 1 0.8991 0 1>}
  NamedRerouteDeclaration(None=@n88.X Softness) {name=Mask X Softness, nodeColor=<rgba 0.0096 1 0.0576 1>}
  NamedRerouteDeclaration(None=@n88.Y Softness) {name=Mask Y Softness, nodeColor=<rgba 1 0 0.6434 1>}
  n89 = ScalarParameter() {defaultValue=1, group=Fake Light Settings, parameterName=Opacity Mask Intensity, sliderMax=1}
  NamedRerouteDeclaration(None=Multiply(A=Saturate(None=@n89), B=NamedRerouteUsage())) {name=Opacity Mask, nodeColor=<rgba 1 0 0.4687 1>}
  ObjectOrientation()
  n90 = NamedRerouteUsage()
  n91 = TextureCoordinate()
  n92 = Constant2Vector() {g=-1, r=-0.5}
  n93 = NamedRerouteUsage()
  n94 = Constant4Vector() {constant=<rgba 0 1 0 0.1>}
  n95 = MaterialFunctionCall(Float3=AppendVector(
    A=Add(A=@n91, B=@n92) {constB=-0.5}
    B=MaterialFunctionCall(Input=@n93, Input Low=@n94.R, Input High=@n94.R, Target Low=@n94.R, Target High=@n94.R) {materialFunction=&RemapValueRange}.Result
   )) {materialFunction=&BreakOutFloat3Components}
  n96 = Length(None=AppendVector(A=@n95.R, B=@n95.R))
  n97 = Max(A=Subtract(A=Multiply(A=@n90) {constB=-1}, B=@n95.G), B=DotProduct(A=NamedRerouteUsage(), B=AppendVector(A=@n96, B=@n95.G)))
  Saturate(None=MaterialFunctionCall(Input=Saturate(None=@n97), Input High=NamedRerouteUsage()) {materialFunction=&RemapValueRange}.Result)
  Saturate(None=@n40.Cone Falloff)
  ScalarParameter() {defaultValue=0.5, group=Surface, parameterName=Surface Specular, sliderMax=1, sortPriority=1}
  ScalarParameter() {defaultValue=0.5, group=Surface, parameterName=Surface Metallic, sliderMax=1, sortPriority=3}
  ScalarParameter() {defaultValue=0.5, group=Surface, parameterName=Surface Roughness, sliderMax=1, sortPriority=0}
 ]
}
